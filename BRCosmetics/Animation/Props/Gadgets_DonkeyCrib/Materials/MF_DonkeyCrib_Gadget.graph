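# Material function for DonkeyCrib gadget props that optionally draws progressively revealed chalk board text into the base color of incoming material attributes.
# Selectable channels of an FX mask input supply a text shape mask and a flow mask.
# The flow mask, offset by TextFlowProgress and smoothstepped, gates where TextChalkColor blends over the original color.
# The UseBoardText static switch chooses between the modified attributes and a pass-through.
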MaterialFunction /BRCosmetics/Animation/Props/Gadgets_DonkeyCrib/Materials/MF_DonkeyCrib_Gadget
graph {nodes [
  Frac(None=Time())
  n1 = GetMaterialAttributes(None=NamedRerouteDeclaration(None=FunctionInput() {inputType=FunctionInput_MaterialAttributes}) {name=InMaterial, nodeColor=<rgba 1 0 0.6797 1>}) {attributeGetTypes [#1]}
  n2 = NamedRerouteUsage()
  FunctionOutput(None=StaticSwitchParameter(
    True=SetMaterialAttributes!(MaterialAttributes=@n1.MaterialAttributes, BaseColor=@n2) {attributeSetTypes [#1]}
    False=NamedRerouteUsage()
   ) {parameterName=UseBoardText})
  n3 = FunctionInput(Preview=TextureSample() {samplerType=SAMPLERTYPE_Masks, texture=/BRCosmetics/Animation/Props/Gadgets_DonkeyCrib/Textures/T_DonkeyCrib_Emote_Prop_FX_Alpha}.RGBA) {inputName=InFXMask, inputType=FunctionInput_Vector4, sortPriority=1}
  NamedRerouteDeclaration(None=ChannelMaskParameter(None=@n3) {maskChannel=Alpha, parameterName=TextFlowMask, defaultValue=<rgba 0 0 0 1>}.R) {name=TextFlowMask, nodeColor=<rgba 0.5234 1 0 1>}
  NamedRerouteDeclaration(None=@n1.BaseColor) {name=InBaseColor, nodeColor=<rgba 1 0.7266 0 1>}
  NamedRerouteDeclaration(None=ChannelMaskParameter(None=@n3) {parameterName=TextFXMask}.R) {name=InTextFXMask, nodeColor=<rgba 0.5234 1 0 1>}
  n4 = NamedRerouteUsage()
  n5 = OneMinus(None=NamedRerouteUsage())
  n6 = Add(A=Subtract(A=@n5), B=Saturate(None=ScalarParameter() {parameterName=TextFlowProgress}))
  n7 = Max(A=Floor(None=@n5), B=SmoothStep(Value=Saturate(None=@n6)) {constMax=0.1})
  NamedRerouteDeclaration(None=LinearInterpolate(
    A=NamedRerouteUsage()
    B=VectorParameter() {parameterName=TextChalkColor, defaultValue=<rgba 0.9888 1 0.8792 1>}.RGB
    Alpha=Multiply(A=@n4, B=@n7)
   )) {name=OutBaseColor, nodeColor=<rgba 0.4062 0 1 1>}
 ]}
