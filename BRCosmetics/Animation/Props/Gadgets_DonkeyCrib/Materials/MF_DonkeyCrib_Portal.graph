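# Material function that turns the incoming material attributes into a glowing, swirling portal for the DonkeyCrib gadget, behind the UsePortal static switch.
# Ray-cylinder intersection and a custom raymarched UV loop fake interior depth, driving panned, distorted noise textures.
# Noise, depth and face-side masks blend edge and core colors into emissive.
# Opacity combines a flickering sphere mask, custom-depth cutout and PortalOpacity, then passes through a translucency-sorting function.
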MaterialFunction /BRCosmetics/Animation/Props/Gadgets_DonkeyCrib/Materials/MF_DonkeyCrib_Portal
graph {nodes [
  ActorPositionWS()
  n1 = NamedRerouteUsage()
  n2 = NamedRerouteUsage()
  Add(A=@n1, B=Multiply(A=Subtract(A=@n2) {constB=0.5}) {constB=0.1})
  n3 = NamedRerouteUsage()
  Ceil(None=@n3)
  ConstantBiasScale(None=NamedRerouteUsage())
  n4 = Multiply(A=NamedRerouteUsage()) {constB=0.01}
  n5 = MaterialFunctionCall(Color=@n4) {materialFunction=&SplitComponents}
  Divide(A=Arctangent2Fast(Y=@n5.R, X=@n5.R), B=MaterialFunctionCall() {materialFunction=&Pi}.Result)
  Frac(None=Time())
  n6 = NamedRerouteUsage()
  n7 = Constant()
  n8 = SetMaterialAttributes!(
   MaterialAttributes=NamedRerouteDeclaration(None=FunctionInput() {inputType=FunctionInput_MaterialAttributes}) {name=InMaterial, nodeColor=<rgba 1 0 0.6797 1>}
   EmissiveColor=NamedRerouteUsage()
   OpacityMask=@n6
   Opacity=@n6
   Roughness=Constant() {r=1}
   Metallic=@n7
   Specular=@n7
   BaseColor=@n7
  ) {attributeSetTypes [
    #1
    #2
    #3
    #4
    #5
    #6
    #7
   ]}
  FunctionOutput(None=StaticSwitchParameter(True=@n8, False=NamedRerouteUsage()) {parameterName=UsePortal})
  n9 = MaterialFunctionCall(Float2=NamedRerouteUsage()) {materialFunction=&BreakOutFloat2Components}
  n10 = Multiply(
   A=AppendVector(A=@n9.R, B=@n9.R)
   B=ComponentMask(None=VectorParameter() {parameterName=UVTiling, defaultValue=<rgba 2 20 0 1>}.RGB) {g=true, r=true}
  )
  n11 = TextureSample!(
   UVs=Panner(Coordinate=@n10) {speedX=-0.1, speedY=-0.1}
   Tex
   "DDX(UVs)"=NamedRerouteUsage()
   "DDY(UVs)"=NamedRerouteUsage()
   Apply View MipBias
  ) {
   mipValueMode=TMVM_Derivative
   samplerType=SAMPLERTYPE_LinearColor
   texture=/BRCosmetics/Characters/Player/Male/Medium/Bodies/M_MED_EggnogFaucet/Materials/T_EggnogFaucet_mask
  }
  n12 = Add(A=Multiply(A=@n10) {constB=0.75}, B=Multiply(A=Subtract(A=@n11.G) {constB=0.5}) {constB=0.1})
  n13 = TextureSample(UVs=Panner(Coordinate=@n12) {speedX=-0.1, speedY=0.5}, "DDX(UVs)"=NamedRerouteUsage(), "DDY(UVs)"=NamedRerouteUsage()) {
   mipValueMode=TMVM_Derivative
   samplerType=SAMPLERTYPE_LinearColor
   texture=&T_EggnogFaucet_mask
  }
  n14 = TextureSample(
   UVs=Panner(Coordinate=Multiply(A=@n12) {constB=0.75}) {speedX=-0.2, speedY=0.25}
   "DDX(UVs)"=NamedRerouteUsage()
   "DDY(UVs)"=NamedRerouteUsage()
  ) {
   mipValueMode=TMVM_Derivative
   samplerType=SAMPLERTYPE_LinearColor
   texture=&T_EggnogFaucet_mask
  }
  LinearInterpolate(A=@n13.G, B=@n14.G)
  n15 = NamedRerouteUsage()
  n16 = OneMinus(None=NamedRerouteUsage())
  n17 = Multiply(
   A=Power(Base=@n16, Exp=ScalarParameter() {defaultValue=2, parameterName=sdfgsdf}) {constExponent=0.5}
   B=ScalarParameter() {defaultValue=2, parameterName=dfg}
  )
  LinearInterpolate(
   A=VectorParameter() {parameterName=EdgeColor, defaultValue=<rgba 0.0365 0.2153 1 1>}.RGB
   B=VectorParameter() {parameterName=CoreColor, defaultValue=<rgba 0.2024 3.1047 5 1>}.RGB
   Alpha=Add(A=@n15, B=Saturate(None=@n17))
  )
  n18 = Saturate(None=Power(Base=NamedRerouteUsage(), Exp=ScalarParameter() {defaultValue=1, parameterName=efgdfs}) {constExponent=0.5})
  n19 = NamedRerouteUsage()
  LinearInterpolate(A=Multiply(B=LinearInterpolate(Alpha=@n18) {constA=0.1}), Alpha=@n19) {constB=0}
  n20 = AppendVector(A=@n5.R, B=@n5.B)
  n21 = MaterialFunctionCall(Color=@n4) {materialFunction=&SplitComponents}
  LinearInterpolate(
   A=@n20
   B=MaterialFunctionCall(Vector or UVs=AppendVector(A=@n21.R, B=@n21.R)) {materialFunction=&VectorToRadialValue}.Radial Coordinates
   Alpha=Saturate(None=Power(Base=@n3) {constExponent=16})
  )
  n22 = MaterialFunctionCall!(UV TexCoord=TextureCoordinate(), Time, Speed, Spin) {materialFunction=/Game/Effects/Niagara/MasterMaterials/Functions/Mojo_MF/MF_FX_Swirl_UV}
  n23 = TextureSample!(UVs=Panner(Coordinate=@n22.Out UV TexCoord) {speedX=0.15, speedY=0.05}, Tex, Apply View MipBias) {
   samplerType=SAMPLERTYPE_LinearColor
   texture=&T_EggnogFaucet_mask
  }
  Multiply(A=@n23.G, B=@n23.G)
  Multiply(A=@n9.G, B=TwoSidedSign())
  n24 = ConstantBiasScale(None=NamedRerouteUsage())
  n25 = TextureSample(UVs=Panner(Coordinate=@n24) {speedX=0.075, speedY=0.075}) {
   samplerType=SAMPLERTYPE_LinearColor
   texture=&T_EggnogFaucet_mask
  }
  Multiply(A=@n25.G, B=@n25.G)
  n26 = TextureSample(UVs=Panner(Coordinate=@n24) {speedX=0.15, speedY=0.05}) {
   samplerType=SAMPLERTYPE_LinearColor
   texture=&T_EggnogFaucet_mask
  }
  Multiply(A=@n26.G, B=@n26.G)
  n27 = Multiply(A=@n13.G, B=@n14.B)
  Multiply(A=@n27, B=OneMinus(None=Multiply(A=@n19, B=@n19)))
  n28 = Subtract(A=NamedRerouteUsage()) {constB=0.5}
  n29 = Constant() {r=0.05}
  n30 = ConstantBiasScale(None=MaterialFunctionCall!(FlickerInitialSpeed=@n29, FlickerRandom=@n29) {materialFunction=/Game/Effects/Fort_Effects/Materials/_Functions/MF_Flicker}.Result)
  n31 = SphereMask(
   A=Add(A=NamedRerouteUsage(), B=Multiply(A=@n28) {constB=0.1})
   B=Constant()
   Radius=Subtract(A=ScalarParameter() {defaultValue=50, parameterName=Radius}, B=Multiply(A=@n30) {constB=0.1})
  ) {attenuationRadius=1, hardnessPercent=0}
  Multiply(A=LinearInterpolate(B=@n15, Alpha=@n16) {constA=1}, B=SmoothStep(Value=@n31) {constMax=0.3})
  n32 = TextureCoordinate()
  NamedRerouteDeclaration(None=DDX(Value=@n32)) {name=DDX, nodeColor=<rgba 1 0.6797 0 1>}
  n33 = NamedRerouteUsage()
  n34 = NamedRerouteUsage()
  n35 = Add(A=ScalarParameter() {defaultValue=1.5}, B=Multiply(A=Power(Base=@n34) {constExponent=0.75}) {constB=0.1})
  NamedRerouteDeclaration(None=Saturate(None=Power(Base=@n33, Exp=@n35))) {name=CylBottom, nodeColor=<rgba 0.1016 0 1 1>}
  n36 = TextureSample(UVs=Panner(Coordinate=@n22.Out UV TexCoord) {speedX=0.075, speedY=0.075}) {
   samplerType=SAMPLERTYPE_LinearColor
   texture=&T_EggnogFaucet_mask
  }
  NamedRerouteDeclaration(None=Multiply(A=@n36.G, B=@n36.G)) {name=Swirl, nodeColor=<rgba 0.4062 0 1 1>}
  n37 = ConstantBiasScale(None=TextureCoordinate()) {bias=-0.5, scale=2}
  n38 = Constant()
  NamedRerouteDeclaration(None=LinearInterpolate(A=NamedRerouteUsage(), B=NamedRerouteDeclaration(None=@n37) {name=UVCentered, nodeColor=<rgba 0.6875 0 1 1>}, Alpha=Distance(A=@n37, B=@n38))) {name=UVDepth, nodeColor=<rgba 0.8516 0 1 1>}
  MaterialExpressionNamedRerouteUsage_20 = NamedRerouteUsage()
  MaterialExpressionNormalize_1 = Normalize(VectorInput=Transform(None=CameraVectorWS()) {transformSourceType=TRANSFORMSOURCE_World, transformType=TRANSFORM_Tangent})
  MaterialExpressionScalarParameter_8 = ScalarParameter() {defaultValue=2, parameterName=RM_Samples}
  MaterialExpressionScalarParameter_7 = ScalarParameter() {defaultValue=1.3463, parameterName=RM_RayLength}
  n39 = Custom!(
   inUV=@MaterialExpressionNamedRerouteUsage_20
   cameraVector=@MaterialExpressionNormalize_1
   steps=@MaterialExpressionScalarParameter_8
   rayLength=@MaterialExpressionScalarParameter_7
  ) {
   code="float3 uvs = float3(inUV.x, inUV.y, 0);\r\n\r\n//dist += distance(uvs.xy/rayLength,0);\r\n// Iteratively push UV2 inward\r\nfor (int i = 0; i < steps; i++){\r\n    uvs -= cameraVector*(rayLength/steps);\r\n    //dist += distance(uvs.xy/rayLength,0);\r\n    dist += distance(uvs.xy/(rayLength),0);\r\n}\r\ndist /= steps;\r\nreturn uvs.xy;"
   outputType=CMOT_Float2
   showCode=true
   additionalOutputs [{outputName=dist, outputType=CMOT_Float1}]
   inputs [
    {
     inputName=inUV
     input {expression=&self:MaterialExpressionNamedRerouteUsage_20}
    }
    {
     inputName=cameraVector
     input {expression=&self:MaterialExpressionNormalize_1}
    }
    {
     inputName=steps
     input {expression=&self:MaterialExpressionScalarParameter_8}
    }
    {
     inputName=rayLength
     input {expression=&self:MaterialExpressionScalarParameter_7}
    }
   ]
  }
  NamedRerouteDeclaration(None=Saturate(None=@n39.dist)) {name=UVDistance, nodeColor=<rgba 0 1 0.1094 1>}
  NamedRerouteDeclaration(None=@n39.return) {name=RaymarchedUVs, nodeColor=<rgba 0 0.2031 1 1>}
  n40 = NamedRerouteDeclaration(None=VectorParameter() {parameterName=PortalCenterPosition, defaultValue=<rgba 0 0 0 1>}.RGB) {name=PortalCenterPos, nodeColor=<rgba 0.2656 1 0 1>}
  n41 = NamedRerouteDeclaration(None=ScalarParameter() {defaultValue=16.5, group=Cylinder, parameterName=ObjecRadiusRelative}) {name=Radius, nodeColor=<rgba 0.5469 0 1 1>}
  n42 = Normalize(VectorInput=VectorParameter() {parameterName=PortalBoneOrientation, defaultValue=<rgba 0 0 1 1>}.RGB)
  n43 = ScalarParameter() {defaultValue=1, parameterName=IsPortalFacingUp}
  n44 = NamedRerouteDeclaration(None=LinearInterpolate(A=@n42, B=Multiply(A=@n42) {constB=-1}, Alpha=@n43)) {name=PortalDirection, nodeColor=<rgba 0.8516 1 0 1>}
  n45 = StaticBool() {value=true}
  n46 = ScalarParameter() {defaultValue=35, parameterName=Depth}
  n47 = AppendVector(A=Multiply(A=NamedRerouteDeclaration(None=@n46) {name=Depth, nodeColor=<rgba 1 0 0.0703 1>}) {constB=-1}, B=Constant())
  n48 = Transform(None=Subtract(
    A=MaterialFunctionCall!(Origin=@n40, Radius=@n41, Z Axis=@n44, XY Scale, X Axis, Capped?=@n45, Z Min-Max=@n47) {materialFunction=/Engine/Functions/Engine_MaterialFunctions01/Math/CylinderIntersection}.Ray Exit Position
    B=@n40
   )) {transformSourceType=TRANSFORMSOURCE_World, transformType=TRANSFORM_Local}
  n49 = RotateAboutAxis(NormalizedRotationAxis=Constant3Vector() {constant=<rgba 0 0 1 0>}, RotationAngle=Constant() {r=0.2}, PivotPoint=Constant(), Position=@n48)
  n50 = ScalarParameter() {parameterName=Min}
  n51 = ScalarParameter() {defaultValue=1, parameterName=Max}
  n52 = Abs(None=Divide(A=ComponentMask(None=@n48) {b=true}, B=NamedRerouteUsage()))
  NamedRerouteDeclaration(None=LinearInterpolate(A=@n48, B=Add(A=@n49, B=@n48), Alpha=LinearInterpolate(A=@n50, B=@n51, Alpha=@n52))) {name=SwirldCylinderCoords, nodeColor=<rgba 0.7344 0 1 1>}
  NamedRerouteDeclaration(None=DDY(Value=@n32)) {name=DDY, nodeColor=<rgba 0.3594 0 1 1>}
  NamedRerouteDeclaration(None=@n20) {name=UVCoords, nodeColor=<rgba 0 0.25 1 1>}
  NamedRerouteDeclaration(None=@n52) {name=DepthMask, nodeColor=<rgba 0.4297 0 1 1>}
  NamedRerouteDeclaration(None=@n27) {name=RawNoise, nodeColor=<rgba 0.5703 0 1 1>}
  NamedRerouteDeclaration(None=@n48) {name=CylinderCoords, nodeColor=<rgba 1 0.2109 0 1>}
  n53 = NamedRerouteUsage()
  n54 = TextureCoordinate()
  n55 = TextureSample!(
   UVs=Panner(Coordinate=MaterialFunctionCall(Vector or UVs=@n53) {materialFunction=&VectorToRadialValue}.Radial Coordinates) {speedX=0.1, speedY=-0.2}
   Tex
   "DDX(UVs)"=DDX(Value=@n54)
   "DDY(UVs)"=DDY(Value=@n54)
   Apply View MipBias
  ) {
   mipValueMode=TMVM_Derivative
   samplerType=SAMPLERTYPE_LinearColor
   texture=&T_EggnogFaucet_mask
  }
  NamedRerouteDeclaration(None=@n55.G) {name=LowFrequencyVoronoi, nodeColor=<rgba 0.5937 1 0 1>}
  NamedRerouteDeclaration(None=@n43) {name=IsPortalFacingUp, nodeColor=<rgba 1 0.9844 0 1>}
  n56 = NamedRerouteUsage()
  n57 = Power(Base=NamedRerouteUsage()) {constExponent=8}
  n58 = Saturate(None=TwoSidedSign())
  n59 = NamedRerouteDeclaration(None=LinearInterpolate(A=OneMinus(None=@n58), B=@n58, Alpha=NamedRerouteUsage())) {name=PortalUp, nodeColor=<rgba 1 0.75 0 1>}
  n60 = LinearInterpolate(
   A=VectorParameter() {parameterName=EdgeColor, defaultValue=<rgba 0.0365 0.2153 1 1>}.RGB
   B=VectorParameter() {parameterName=EdgeColor2, defaultValue=<rgba 0.2713 1 0.8921 1>}.RGB
   Alpha=LinearInterpolate(A=Min(A=@n56, B=@n57), B=NamedRerouteUsage(), Alpha=@n59)
  )
  n61 = VectorParameter() {parameterName=CoreColor, defaultValue=<rgba 0.2024 3.1047 5 1>}
  n62 = SmoothStep(Value=NamedRerouteUsage()) {constMax=0.5}
  n63 = Multiply(A=LinearInterpolate(A=NamedRerouteUsage(), Alpha=@n62), B=@n62)
  n64 = Max(A=LinearInterpolate(B=NamedRerouteUsage(), Alpha=NamedRerouteUsage()), B=Multiply(A=@n59, B=Saturate(None=@n63)))
  n65 = NamedRerouteUsage()
  NamedRerouteDeclaration(None=LinearInterpolate(A=LinearInterpolate(A=@n60, B=@n61.RGB, Alpha=@n64), B=Saturate(None=@n65), Alpha=ScalarParameter() {parameterName=DebugUVs})) {name=OutEmissive, nodeColor=<rgba 0 1 1 1>}
  n66 = ScalarParameter() {defaultValue=1.05, parameterName=TopMaskRadius}
  n67 = SphereMask(A=@n1, B=Constant(), Radius=@n66) {attenuationRadius=1, hardnessPercent=0}
  n68 = SmoothStep(
   Min=NamedRerouteUsage()
   Value=SmoothStep(Value=NamedRerouteDeclaration(None=@n67) {name=Distance, nodeColor=<rgba 1 0 0.3281 1>}) {constMax=0.2}
  )
  n69 = SceneTexture() {sceneTextureId=PPI_CustomDepth}
  n70 = Power(
   Base=SphereMask(A=ComponentMask(None=@n69.Color) {r=true}, B=PixelDepth().R, Radius=NamedRerouteUsage()) {attenuationRadius=15.5, hardnessPercent=0}
   Exp=ScalarParameter() {defaultValue=1, parameterName=DepthFallof}
  )
  n71 = Max(
   A=NamedRerouteUsage()
   B=MaterialFunctionCall(
    Origin=NamedRerouteUsage()
    Radius=NamedRerouteUsage()
    Z Axis=NamedRerouteUsage()
    Capped?=StaticBool() {value=true}
    Z Min-Max=Constant2Vector()
   ) {materialFunction=&CylinderIntersection}.Opacity Mask
  )
  n72 = Multiply(A=Saturate(None=SmoothStep(Value=@n68)), B=OneMinus(None=Multiply(A=@n70, B=@n71)))
  n73 = Saturate(None=ScalarParameter() {defaultValue=1, parameterName=PortalOpacity})
  NamedRerouteDeclaration(None=MaterialFunctionCall!(Opacity=Multiply(A=@n72, B=@n73)) {materialFunction=/Game/Global/Materials/MaterialGym/Materials/SortedOutline/MF_SortTranslucency}.Result) {name=OutOpacity, nodeColor=<rgba 1 0.6797 0 1>}
  NamedRerouteDeclaration(None=@n55.B) {name=LowFrequencyPerlin, nodeColor=<rgba 0.9922 1 0 1>}
  NamedRerouteUsage()
  OneMinus(None=@n39.dist)
  Power(Base=NamedRerouteUsage(), Exp=ScalarParameter() {defaultValue=1, parameterName=RM_LerpExponent})
  Saturate(None=TwoSidedSign())
  SmoothStep(Min=ScalarParameter() {parameterName=min}, Max=ScalarParameter() {defaultValue=1, parameterName=max}, Value=@n33) {constMax=0.5}
  SmoothStep(Min=NamedRerouteUsage(), Value=SmoothStep(Value=@n31) {constMax=0.3})
  n74 = Constant() {r=0.075}
  n75 = ConstantBiasScale(None=NamedRerouteDeclaration(None=MaterialFunctionCall(FlickerInitialSpeed=@n74, FlickerRandom=@n74) {materialFunction=&MF_Flicker}.Result) {name=SlowFlicker, nodeColor=<rgba 0.4766 1 0 1>})
  Subtract(A=@n66, B=NamedRerouteDeclaration(None=Multiply(A=@n75) {constB=0.1}) {name=RadiusGrow, nodeColor=<rgba 1 0 0.2812 1>})
  VectorParameter() {group=Cylinder, parameterName=ZMin-Max, channelNames {a="", b="", g=ZMax, r=ZMin}, defaultValue=<rgba -100 100 0 1>}
 ]}
