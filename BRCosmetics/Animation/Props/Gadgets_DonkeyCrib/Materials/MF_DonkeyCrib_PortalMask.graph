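# Material function for the DonkeyCrib gadget prop that cuts an opacity mask around a portal, behind the UsePortalMasking static switch.
# World position in portal space is combined with a capped cylinder intersection, built from center, bone orientation, radius and a facing flip.
# Camera height in portal space blends between min and max combinations, so masking depends on the viewing side.
# An FX mask channel and IsPortalShowing gate the result written to OpacityMask.
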MaterialFunction /BRCosmetics/Animation/Props/Gadgets_DonkeyCrib/Materials/MF_DonkeyCrib_PortalMask
graph {nodes [
  n1 = NamedRerouteDeclaration(None=FunctionInput() {inputType=FunctionInput_MaterialAttributes}) {name=InMaterial, nodeColor=<rgba 1 0 0.6797 1>}
  n2 = NamedRerouteUsage()
  FunctionOutput(None=StaticSwitchParameter(
    True=SetMaterialAttributes!(MaterialAttributes=@n1, OpacityMask=@n2, EmissiveColor) {attributeSetTypes [#1, #2]}
    False=NamedRerouteUsage()
   ) {parameterName=UsePortalMasking})
  NamedRerouteDeclaration(None=VectorParameter() {parameterName=PortalCenterPosition, defaultValue=<rgba 0 0 0 1>}.RGB) {name=PortalCenter, nodeColor=<rgba 0 0.5312 1 1>}
  n3 = Normalize(VectorInput=VectorParameter() {parameterName=PortalBoneOrientation, defaultValue=<rgba 0 0 1 1>}.RGB)
  NamedRerouteDeclaration(None=LinearInterpolate(A=@n3, B=Multiply(A=@n3) {constB=-1}, Alpha=ScalarParameter() {defaultValue=1, parameterName=IsPortalFacingUp})) {name=PortalDirection, nodeColor=<rgba 0.8516 1 0 1>}
  n4 = ComponentMask(None=MaterialFunctionCall!(
    Vector to Transform=WorldPosition() {worldPositionShaderOffset=WPT_ExcludeAllShaderOffsets}.XYZ
    Z Vector=NamedRerouteUsage()
    Center Location=NamedRerouteUsage()
   ) {materialFunction=/Engine/Functions/Engine_MaterialFunctions02/Math/TransformToZVector}.Result) {b=true}
  n5 = MaterialFunctionCall!(
   Origin=NamedRerouteUsage()
   Radius=ScalarParameter() {defaultValue=16, parameterName=PortalRadius}
   Z Axis=NamedRerouteUsage()
   XY Scale
   X Axis
   Capped?=StaticBool() {value=true}
   Z Min-Max=Constant2Vector()
  ) {materialFunction=/Engine/Functions/Engine_MaterialFunctions01/Math/CylinderIntersection}
  n6 = Min(A=@n4, B=OneMinus(None=@n5.Opacity Mask))
  n7 = Max(A=@n4, B=@n5.Opacity Mask)
  n8 = CameraPositionWS()
  n9 = NamedRerouteUsage()
  n10 = NamedRerouteUsage()
  n11 = Clamp(None=ComponentMask(None=MaterialFunctionCall(Vector to Transform=@n8, Z Vector=@n9, Center Location=@n10) {materialFunction=&TransformToZVector}.Result) {b=true})
  n12 = NamedRerouteUsage()
  n13 = ScalarParameter() {parameterName=IsPortalShowing}
  NamedRerouteDeclaration(None=LinearInterpolate(B=LinearInterpolate(A=@n6, B=@n7, Alpha=@n11), Alpha=Multiply(A=@n12, B=@n13)) {constA=1}) {name=OutOpacityMask, nodeColor=<rgba 0.8281 0 1 1>}
  NamedRerouteDeclaration(None=ChannelMaskParameter(None=FunctionInput() {inputName=InFXMask, sortPriority=1}) {maskChannel=Blue, parameterName=MugFXMask, defaultValue=<rgba 0 0 1 0>}.R) {name=MugMask, nodeColor=<rgba 0.5234 1 0 1>}
 ]}
