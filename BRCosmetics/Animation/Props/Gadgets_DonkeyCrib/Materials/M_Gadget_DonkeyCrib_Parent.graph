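# Parent material for the DonkeyCrib cosmetic gadget prop, built for skeletal meshes with morph targets and mesh deformers.
# It uses material attributes with an expression-driven shading model, wrapping gadget effects between the shared 2023 character PreFX and PostFX functions.
# Between them, Portal, PortalMask and Gadget functions run in sequence.
# PreFX's Tech Art Mask feeds PortalMask, and is combined with its alpha for Gadget.
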Material /BRCosmetics/Animation/Props/Gadgets_DonkeyCrib/Materials/M_Gadget_DonkeyCrib_Parent
graph {
 properties {
  bUseMaterialAttributes=true
  bUsedWithMeshDeformer=true
  bUsedWithMorphTargets=true
  bUsedWithSkeletalMesh=true
  pixelDepthOffsetMode=PDOM_Legacy
  shadingModel=MSM_FromMaterialExpression
  translucencyPass=MTP_AfterMotionBlur
  usedShadingModels=Default Lit | Subsurface | Clear Coat | Subsurface Profile | Cloth
 }
 nodes [
  n1 = MaterialFunctionCall() {materialFunction=&MF_CharacterParent_2023_PreFX}
  n2 = MaterialFunctionCall!(In=@n1.Result) {materialFunction=&/MF_DonkeyCrib_Portal}
  MaterialFunctionCall(
   Material=MaterialFunctionCall!(
    In=MaterialFunctionCall!(In=@n2.Result, InFXMask=@n1.Tech Art Mask) {materialFunction=&/MF_DonkeyCrib_PortalMask}.Result
    InFXMask=AppendVector(A=@n1.Tech Art Mask, B=@n1.Tech Art Mask Alpha)
   ) {materialFunction=&/MF_DonkeyCrib_Gadget}.Result
   M=@n1.M
   Pre Skinned Local Position=@n1.M
   Distance Blend=@n1.M
   Fresnel=@n1.M
  ) {materialFunction=&MF_CharacterParent_2023_PostFX}
 ]
}
